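# Physics asset for TnTina's juggling ball, stored with the Season 12 Battle Pass main player menu animations.
# It holds a single body on the root bone, shaped as one short capsule centered almost exactly on the bone and rotated relative to it.
# It defines no constraints, so there is no articulated chain, just the one collision body.
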PhysicsAsset /BRCosmetics/Animation/Game/MainPlayer/Menu/BR/S12_BP/TnTina/JugglingBall_TnTina_Physics
bodies {root [{shapeType=Capsule, length=4.0865, radius=4.0051, center=<-0.0018 -0.009 -0.0057>, rotation=<rot 88.5498 68.3053 158.3116>}]}
constraints []
